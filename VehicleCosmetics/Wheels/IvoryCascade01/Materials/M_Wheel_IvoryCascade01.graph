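# Masked, default-lit master material for the IvoryCascade01 vehicle cosmetic wheel, built on MF_WheelParent_Simple for skeletal meshes.
# An optional Enable_ColorD static switch adds a fourth rim color, overriding base color, roughness and metallic through the rim mask's alpha channel.
# Red vertex color restores the parent's values over that override.
# The result passes through MF_BaseVehicle_Damage, then MF_VehicleDeRez_Transition.
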Material /VehicleCosmetics/Wheels/IvoryCascade01/Materials/M_Wheel_IvoryCascade01
graph {
 properties {
  bAutomaticallySetUsageInEditor=false
  bUseMaterialAttributes=true
  bUsedWithMeshDeformer=true
  bUsedWithSkeletalMesh=true
  blendMode=BLEND_Masked
  pixelDepthOffsetMode=PDOM_Legacy
  usedShadingModels=Default Lit
 }
 nodes [
  n1 = MaterialFunctionCall() {materialFunction=&MF_WheelParent_Simple}
  n2 = GetMaterialAttributes(None=@n1.MaterialAttributes) {attributeGetTypes [#1, #2, #3]}
  n3 = VectorParameter() {group=6_Additional Properties, parameterName=Rim_ColorD, sortPriority=1, defaultValue=<rgba 0 0 0 1>}
  n4 = ComponentMask(None=@n1.RimRGBAMask) {a=true}
  n5 = VectorParameter() {
   group=6_Additional Properties
   parameterName=Rim_ColorD_PBR
   sortPriority=2
   channelNames {a=Unuse, b=Metallic, g=Roughness, r=Unuse}
   defaultValue=<rgba 0 1 1 1>
  }
  n6 = SetMaterialAttributes!(
   MaterialAttributes=@n2.MaterialAttributes
   Base Color=LinearInterpolate(A=LinearInterpolate(A=@n2.Base Color, B=@n3.RGB, Alpha=@n4), B=@n2.Base Color, Alpha=@n1.VertexColorR)
   Roughness=LinearInterpolate(A=LinearInterpolate(A=@n2.Roughness, B=@n5.Roughness, Alpha=@n4), B=@n2.Roughness, Alpha=@n1.VertexColorR)
   Metallic=LinearInterpolate(A=LinearInterpolate(A=@n2.Metallic, B=@n5.Metallic, Alpha=@n4), B=@n2.Metallic, Alpha=@n1.VertexColorR)
  ) {attributeSetTypes [#1, #2, #3]}
  MaterialFunctionCall(In=MaterialFunctionCall(Material=StaticSwitchParameter(True=@n6, False=@n1.MaterialAttributes) {group=6_Additional Properties, parameterName=Enable_ColorD, sortPriority=0}) {materialFunction=&MF_BaseVehicle_Damage}.Result) {materialFunction=&MF_VehicleDeRez_Transition}
 ]
}
